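# Animation Blueprint for the Commando Medic body cosmetic, adding procedural secondary motion to the pose copied from its attached parent mesh.
# An RBF Pose Driver reads the spine bones to drive dynamic spine and pouch bones through a pose asset.
# The stethoscope bone chains a weighted neck/spine/clavicle constraint, partially blended spring AnimDynamics with a clavicle sphere limit, a left-clavicle constraint and an additive rotation offset.
# The event graph holds no logic.
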AnimBlueprint /BRCosmetics/Characters/Player/Male/Medium/Bodies/M_MED_Commando_Medic/Meshes/M_MED_Medic_Body_AnimBP
target_skeleton=&/M_MED_Medic_Body_Skeleton
graphs [
 AnimGraph {nodes [
   n1 = Pose Driver - Source: spine_01!(SourcePose:Pose Link Structure=Copy Pose From Mesh() {bUseAttachedParent=true}) -> (Pose:Pose Link Structure) {
    is=title AnimGraphNode_PoseDriver
    poseAsset=&/M_MED_Medic_Spine_RBF_PoseAsset
    onlyDriveBones [
     {boneName=dyn_spine_01_driven}
     {boneName=dyn_spine_02_driven}
     {boneName=dyn_spine_03_driven}
     {boneName=dyn_spine_04_driven}
     {boneName=dyn_spine_05_driven}
     {boneName=dyn_pouch}
    ]
    poseTargets [
     {
      bApplyCustomCurve=false
      bIsHidden=false
      distanceMethod=DefaultMethod
      drivenName=Pose_0
      functionType=DefaultFunction
      targetScale=0.8225
      boneTransforms [
       {targetRotation=<rot 0 -3.9009 0>, targetTranslation=<6.7494 -0.6388 0>}
       {targetRotation=<rot 0 -2.7227 0>, targetTranslation=<7.2518 -0.035 0>}
       {targetRotation=<rot 0 -1.7047 0>, targetTranslation=<6.0104 0 0>}
       {targetRotation=<rot 0 1.5078 0>, targetTranslation=<8.1024 0 0>}
       {targetRotation=<rot 0 0.1712 0>, targetTranslation=<10.0146 0 0>}
      ]
      customCurve {defaultValue=3.4028234663852886e+38, postInfinityExtrap=RCCE_Constant, preInfinityExtrap=RCCE_Constant, keys []}
      targetRotation=<rot 0 0 0>
     }
     {
      bApplyCustomCurve=false
      bIsHidden=false
      distanceMethod=DefaultMethod
      drivenName=Pose_1
      functionType=DefaultFunction
      targetScale=0.8225
      boneTransforms [
       {targetRotation=<rot -9.8046 9.1979 4.183>, targetTranslation=<6.7494 -0.6388 0>}
       {targetRotation=<rot -0.9946 1.4714 16.5587>, targetTranslation=<7.2518 -0.035 0>}
       {targetRotation=<rot -0.0758 -1.9943 26.929>, targetTranslation=<6.0104 0 0>}
       {targetRotation=<rot -2.3574 -3.9817 35.5043>, targetTranslation=<8.1024 0 0>}
       {targetRotation=<rot -7.8815 -6.5671 16.7378>, targetTranslation=<10.0146 0 0>}
      ]
      customCurve {defaultValue=3.4028234663852886e+38, postInfinityExtrap=RCCE_Constant, preInfinityExtrap=RCCE_Constant, keys []}
      targetRotation=<rot 0 0 0>
     }
     {
      bApplyCustomCurve=false
      bIsHidden=false
      distanceMethod=DefaultMethod
      drivenName=Pose_2
      functionType=DefaultFunction
      targetScale=1
      boneTransforms [
       {targetRotation=<rot 0 -3.9009 0>, targetTranslation=<6.7494 -0.6388 0>}
       {targetRotation=<rot -5.3951 14.963 -24.4519>, targetTranslation=<7.2518 -0.035 0>}
       {targetRotation=<rot 5.3508 1.6592 -18.9838>, targetTranslation=<6.0104 0 0>}
       {targetRotation=<rot 9.7778 3.9022 -23.3152>, targetTranslation=<8.1024 0 0>}
       {targetRotation=<rot 11.9649 2.7821 -12.2104>, targetTranslation=<10.0146 0 0>}
      ]
      customCurve {defaultValue=3.4028234663852886e+38, postInfinityExtrap=RCCE_Constant, preInfinityExtrap=RCCE_Constant, keys []}
      targetRotation=<rot 0 0 0>
     }
    ]
    rBFParams {radius=20}
    sourceBones [{boneName=spine_01}, {boneName=spine_02}, {boneName=spine_03}, {boneName=spine_04}, {boneName=spine_05}]
   }
   n2 = "Constraint - dyn_stethoscope (PPPPP)"!(
    ComponentPose=Local To Component(LocalPose=@n1)
    Alpha=1
    ConstraintWeights_0=0.5755
    ConstraintWeights_1:float=0.0047
    ConstraintWeights_2:float=0.0151
    ConstraintWeights_3:float=0.3836
    ConstraintWeights_4:float=0.021
   ) -> (Pose) {
    alphaBoolBlend {blendOption=Linear}
    boneToModify {boneName=dyn_stethoscope}
    constraintSetup [
     {targetBone {boneName=neck_02}}
     {targetBone {boneName=neck_01}}
     {targetBone {boneName=spine_04}}
     {targetBone {boneName=spine_05}}
     {targetBone {boneName=clavicle_l}}
    ]
    constraintWeights [1,1,1,1,1]
   }
   n3 = Anim Dynamics - Bone: dyn_stethoscope(ComponentPose=@n2, Alpha=0.25, ExternalForce="0,0,0") {
    angularDampingOverride=0.99
    angularSpringConstant=250
    bAngularSpring=true
    bGravityOverrideInSimSpace=true
    bOverrideAngularDamping=true
    bUseSphericalLimits=true
    gravityScale=0.5
    alphaBoolBlend {blendOption=Linear}
    boundBone {boneName=dyn_stethoscope}
    componentLinearAccScale=<0.5 0.5 0.5>
    componentLinearVelScale=<0.5 0.5 0.5>
    physicsBodyDefinitions [{
      sphereCollisionRadius=0
      boundBone {boneName=dyn_stethoscope}
      boxExtents=<10 1 10>
      constraintSetup {angularLimitsMax=<0 20 20>, angularLimitsMin=<0 -20 2>, angularTarget=<0 0 0>}
      localJointOffset=<5 0 0>
     }]
    sphericalLimits [{limitRadius=15, drivingBone {boneName=clavicle_l}, sphereLocalOffset=<5.5 0 -8>}]
   }
   n4 = "Constraint - dyn_stethoscope (P)"(ComponentPose=@n3, Alpha=1, ConstraintWeights_0=1) {
    alpha=0
    alphaBoolBlend {blendOption=Linear}
    boneToModify {boneName=dyn_stethoscope}
    constraintSetup [{targetBone {boneName=clavicle_l}}]
    constraintWeights [1]
   }
   n5 = "Transform (Modify) Bone - Bone: dyn_stethoscope"(ComponentPose=@n4, Alpha=18, Translation="0,0,0", Rotation="0,8,0", Scale="1,1,1") {
    alpha=0
    rotationMode=BMM_Additive
    rotationSpace=BCS_BoneSpace
    scaleMode=BMM_Additive
    scaleSpace=BCS_BoneSpace
    translationMode=BMM_Additive
    translationSpace=BCS_BoneSpace
    alphaBoolBlend {blendOption=Linear}
    boneToModify {boneName=dyn_stethoscope}
    rotation {yaw=8}
   }
   Output Pose(Result=Component To Local(ComponentPose=@n5)) {layerGroup=None}
  ]}
 EventGraph {nodes [Event BlueprintUpdateAnimation()]}
]
variables []
